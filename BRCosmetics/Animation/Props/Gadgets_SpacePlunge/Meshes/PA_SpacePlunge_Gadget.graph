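# Physics asset for the SpacePlunge gadget prop mesh, found among the BRCosmetics animation props.
# It has two capsule bodies: a small one on the main bone and a far larger, nearly spherical one on the root bone.
# A single constraint joins main to root, with matching swing and twist limit angles.
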PhysicsAsset /BRCosmetics/Animation/Props/Gadgets_SpacePlunge/Meshes/PA_SpacePlunge_Gadget
bodies {
 main [{shapeType=Capsule, length=6.125, radius=9.7115, rotation=<rot 0.0039 -90 0.0015>}]
 root [{shapeType=Capsule, length=0.3247, radius=190.8786, rotation=<rot 0 -90 0>}]
}
constraints [{bone1Name=main, bone2Name=root, swing1LimitDegrees=45, swing2LimitDegrees=45, twistLimitDegrees=45}]
